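# Material function that gives the TalonPane gadget emote prop a translucent hologram look by overriding emissive and opacity.
# Fresnel blends base color toward a see-through environment, apparently the skylight cubemap reroute (time-of-day colors on low quality), then adds tint and sunlight.
# Emissive scales by time of day and brightens under ray tracing.
# Opacity combines a min/max range, scene-depth cutoff and TechArt texture mask, alongside platform-scaled scanlines.
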MaterialFunction /BRCosmetics/Animation/Props/Gadgets_TalonPane/Materials/MF_Hologram_TalonPane
graph {nodes [
  n1 = GetMaterialAttributes(None=FunctionInput() {inputType=FunctionInput_MaterialAttributes}) {attributeGetTypes [#1, #2]}
  n2 = ScalarParameter() {defaultValue=3, parameterName=See Through Side Max}
  n3 = Saturate(None=Fresnel(ExponentIn=ScalarParameter() {defaultValue=3, parameterName=SeethroughSidesFres}) {baseReflectFraction=0})
  n4 = LinearInterpolate(A=Add(A=@n1.BaseColor, B=@n1.BaseColor), B=NamedRerouteUsage(), Alpha=Saturate(None=LinearInterpolate(B=@n2, Alpha=@n3) {constB=0}))
  n5 = NamedRerouteUsage()
  n6 = VectorParameter() {parameterName=ColorTint, defaultValue=<rgba 1 0.4673 0 2>}
  n7 = Multiply(A=Clamp(None=Multiply(A=@n4, B=@n5)), B=Multiply(A=@n6.RGB, B=@n6.RGB))
  n8 = ScalarParameter() {defaultValue=0.02, parameterName=Sunlight Passthrough}
  n9 = CollectionParameter() {collection=&FortniteMaterialParameters, parameterName=SunLightColor}
  n10 = Multiply(A=ComponentMask(None=Saturate(None=@n9)) {b=true, g=true, r=true}, B=NamedRerouteUsage())
  n11 = MaterialFunctionCall() {materialFunction=/Game/Packages/Fortress_Sky/Fortnite_SunScreenSpaceFalloff}
  n12 = QualitySwitch(Default=Multiply(A=Multiply(A=@n8, B=@n10), B=Clamp(None=@n11.Dot*-1) {minDefault=0.5}), Low=Constant())
  n13 = Constant() {r=0.7}
  n14 = Constant() {r=1}
  n15 = Constant() {r=0.4}
  n16 = Add(
   A=@n1.EmissiveColor
   B=Multiply(
    A=Add(A=@n7, B=@n12)
    B=MaterialFunctionCall(Value_Morning=@n13, Value_Day=@n14, Value_Evening=@n13, Value_Night=@n15) {materialFunction=&Scalar_Time_of_Day}.Result
   )
  )
  n17 = Multiply(A=@n16, B=ScalarParameter() {defaultValue=8, parameterName=LumenBrightness})
  n18 = CameraPositionWS()
  n19 = WorldPosition()
  n20 = Add(A=Subtract(A=SceneDepth().R, B=Distance(A=@n18, B=@n19.XYZ)), B=Constant() {r=20})
  n21 = Multiply(
   A=LinearInterpolate(
    A=ScalarParameter() {defaultValue=0.2, group=Translucent, parameterName=MinOpacity}
    B=ScalarParameter() {defaultValue=0.8, group=Translucent, parameterName=MaxOpacity}
    Alpha=NamedRerouteUsage()
   ) {constA=0.2, constB=0.7}
   B=Saturate(None=Sign(None=@n20))
  )
  n22 = TextureSampleParameter2D() {group=Translucent, parameterName=TechArtRE, texture=/BRCosmetics/Animation/Props/Gadgets_TalonPane/Textures/T_Talonpane_Emote_Prop_TechArt}
  n23 = Power(Base=OneMinus(None=Fresnel() {baseReflectFraction=0, exponent=1})) {constExponent=3}
  n24 = Multiply(A=LinearInterpolate(A=LinearInterpolate(Alpha=@n22.R) {constA=0.05}, B=LinearInterpolate(Alpha=@n23) {constA=0.1}, Alpha=@n22.G), B=@n22.B)
  FunctionOutput(None=SetMaterialAttributes!(
    MaterialAttributes=@n1.MaterialAttributes
    EmissiveColor=RayTracingQualitySwitch(Normal=@n16, RayTraced=@n17)
    Opacity=Multiply(A=@n21, B=@n24)
   ) {attributeSetTypes [#2, #3]})
  n25 = Constant() {r=1}
  n26 = MaterialFunctionCall(Value_Morning=@n25, Value_Day=@n25, Value_Evening=@n25, Value_Night=ScalarParameter() {defaultValue=5, parameterName=Night Boost}) {materialFunction=&Scalar_Time_of_Day}
  Multiply(A=@n26.Result) {constB=0.5}
  n27 = ScalarParameter() {defaultValue=20, group=AuraCop, parameterName=ScanlineAmountrRE}
  n28 = ScalarParameter() {defaultValue=60, group=AuraCop, parameterName=ScanlineAmountrRE_Mobile}
  NamedRerouteDeclaration(None=MaterialFunctionCall!(Scanline Amount=MaterialFunctionCall("Xb1,PS4"=@n27, "PC-High&&Epic,XSX,PS5"=@n27, "PC-Perf,Android"=@n28, Switch=@n28, "PC-Low(NotPerf)&&Medium"=@n27) {materialFunction=&PlatformSwitch_1D}.Result) {materialFunction=&/MFScanlines_TalonPane}.Result) {name=Scanlines, nodeColor=<rgba 1 0 0.1641 1>}
  n29 = MaterialFunctionCall(Float3=OneMinus(None=CameraVectorWS())) {materialFunction=&BreakOutFloat3Components}
  n30 = MaterialFunctionCall(X=@n29.R, Y=@n29.R, Z=Add(A=@n29.B, B=ScalarParameter() {defaultValue=-1, parameterName=Offset SeeThrough})) {materialFunction=&MakeFloat3}
  n31 = ScalarParameter() {defaultValue=0.1, parameterName=HologramRoughness}
  n32 = Multiply(
   A=Clamp(None=SkyLightEnvMapSample(Direction=@n30.Result, Roughness=@n31)) {maxDefault=4, minDefault=0.5}
   B=ScalarParameter() {defaultValue=1, parameterName=CubemapBleed}
  )
  n33 = VectorParameter() {parameterName=ColorMorning, defaultValue=<rgba 0.2031 0.1723 0.1315 1>}
  n34 = VectorParameter() {parameterName=ColorNight, defaultValue=<rgba 0.0373 0.0718 0.1302 1>}
  NamedRerouteDeclaration(None=QualitySwitch(
    Default=Multiply(A=@n32, B=@n26.Result)
    Low=MaterialFunctionCall(Color_Morning=@n33.RGB, Color_Day=@n33.RGB, Color_Evening=@n33.RGB, Color_Night=@n34.RGB) {materialFunction=&Color_Time_of_Day}.Result
   )) {name=Cubemap, nodeColor=<rgba 0.8984 0 1 1>}
 ]}
