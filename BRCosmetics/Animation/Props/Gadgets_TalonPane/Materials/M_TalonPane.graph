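# Translucent unlit material for the Talon Pane gadget prop, built for skeletal meshes with morph targets and mesh deformers.
# It chains the character parent PreFX function into a Talon Pane hologram function, then into PostFX on material attributes.
# PreFX's M output also feeds PostFX's mask, pre-skinned position, distance blend and fresnel inputs directly.
# Depth testing is disabled and translucency renders after motion blur.
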Material /BRCosmetics/Animation/Props/Gadgets_TalonPane/Materials/M_TalonPane
graph {
 properties {
  bAllowFrontLayerTranslucency=false
  bDisableDepthTest=true
  bUseMaterialAttributes=true
  bUsedWithMeshDeformer=true
  bUsedWithMorphTargets=true
  bUsedWithSkeletalMesh=true
  blendMode=BLEND_Translucent
  pixelDepthOffsetMode=PDOM_Legacy
  shadingModel=MSM_Unlit
  translucencyLightingMode=TLM_SurfacePerPixelLighting
  translucencyPass=MTP_AfterMotionBlur
  usedShadingModels=Unlit
 }
 nodes [
  n1 = MaterialFunctionCall() {materialFunction=&MF_CharacterParent_2023_PreFX}
  MaterialFunctionCall(
   Material=MaterialFunctionCall!(In=@n1.Result) {materialFunction=&/MF_Hologram_TalonPane}.Result
   M=@n1.M
   Pre Skinned Local Position=@n1.M
   Distance Blend=@n1.M
   Fresnel=@n1.M
  ) {materialFunction=&MF_CharacterParent_2023_PostFX}
 ]
}
